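# Prototype material for straight path pieces in the environment, using the Default Lit shading model.
# It is flagged for use on spline meshes, with automatic usage flagging in the editor turned off.
# The graph holds a single vector parameter defaulting to a neutral grey, with no textures or other logic.
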Material /Hera/Environment/Prototype/ChrisK/Paths/M_Path_Straight_Proto_CK
graph {
 properties {bAutomaticallySetUsageInEditor=false, bUsedWithSplineMeshes=true, usedShadingModels=Default Lit}
 nodes [VectorParameter() {defaultValue=<rgba 0.2031 0.2031 0.2031 0>}]
}
